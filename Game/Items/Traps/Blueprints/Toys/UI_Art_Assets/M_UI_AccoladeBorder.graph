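# UI-domain translucent material for an accolade border, which procedurally builds a rounded-rectangle mask from UVs without textures.
# Ceil tests on U and V, inset by the Border parameter, form two crossing bands that fill everything except the corners.
# Four high-density RadialGradientExponential discs of Border radius round off the corners, and everything is summed and clamped.
# Color and Alpha parameters are exposed for tint and opacity.
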Material /Game/Items/Traps/Blueprints/Toys/UI_Art_Assets/M_UI_AccoladeBorder
graph {
 properties {blendMode=BLEND_Translucent, materialDomain=MD_UI, pixelDepthOffsetMode=PDOM_Legacy, usedShadingModels=Default Lit}
 nodes [
  n1 = Subtract(A=Constant(), B=ScalarParameter() {defaultValue=0.25, parameterName=Border})
  n2 = Add(A=TextureCoordinate(), B=@n1) {constB=-0.125}
  n3 = Add(A=OneMinus(None=TextureCoordinate()), B=@n1) {constB=-0.125}
  n4 = DotProduct(A=Ceil(None=ComponentMask(None=@n2) {r=true}), B=Ceil(None=ComponentMask(None=@n3) {r=true}))
  n5 = Add(A=TextureCoordinate(), B=@n1) {constB=-0.125}
  n6 = Add(A=OneMinus(None=TextureCoordinate()), B=@n1) {constB=-0.125}
  n7 = DotProduct(A=Ceil(None=ComponentMask(None=@n5) {g=true}), B=Ceil(None=ComponentMask(None=@n6) {g=true}))
  n8 = TextureCoordinate()
  n9 = ScalarParameter() {defaultValue=0.25, parameterName=Border}
  n10 = AppendVector(A=@n9, B=@n9)
  n11 = Constant() {r=32}
  n12 = TextureCoordinate()
  n13 = ScalarParameter() {defaultValue=0.25, parameterName=Border}
  n14 = AppendVector(A=Subtract(A=Constant() {r=1}, B=@n13), B=@n13)
  n15 = Constant() {r=32}
  n16 = Add(
   A=Multiply(A=MaterialFunctionCall(UVs=@n8, CenterPosition=@n10, Radius=@n9, Density=@n11) {materialFunction=&RadialGradientExponential}.RadialGradientExponential)
   B=Multiply(A=MaterialFunctionCall(UVs=@n12, CenterPosition=@n14, Radius=@n13, Density=@n15) {materialFunction=&RadialGradientExponential}.RadialGradientExponential)
  )
  n17 = TextureCoordinate()
  n18 = ScalarParameter() {defaultValue=0.25, parameterName=Border}
  n19 = Subtract(A=Constant() {r=1}, B=@n18)
  n20 = AppendVector(A=@n19, B=@n19)
  n21 = Constant() {r=32}
  n22 = TextureCoordinate()
  n23 = ScalarParameter() {defaultValue=0.25, parameterName=Border}
  n24 = AppendVector(A=@n23, B=Subtract(A=Constant() {r=1}, B=@n23))
  n25 = Constant() {r=32}
  n26 = Add(
   A=Multiply(A=MaterialFunctionCall(UVs=@n17, CenterPosition=@n20, Radius=@n18, Density=@n21) {materialFunction=&RadialGradientExponential}.RadialGradientExponential)
   B=Multiply(A=MaterialFunctionCall(UVs=@n22, CenterPosition=@n24, Radius=@n23, Density=@n25) {materialFunction=&RadialGradientExponential}.RadialGradientExponential)
  )
  Clamp(None=Add(A=Add(A=@n4, B=@n7), B=Add(A=@n16, B=@n26)))
  ScalarParameter() {defaultValue=1, parameterName=Alpha}
  VectorParameter() {parameterName=Color, defaultValue=<rgba 0.6719 0.2415 0.2415 1>}
 ]
}
