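# Additive UI material that draws a procedural glowing flare, apparently for accolade displays.
# A radial gradient is screen-blended with a V-squashed horizontal streak, smoothstepped, then added to a small core glow to blend between two color parameters.
# The ControlTime static switch selects the TimeControl parameter or a slow sine of Time to pulse streak, threshold and core.
# UV scale, Grow and Flare nodes are unconnected.
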Material /Game/Items/Traps/Blueprints/Toys/UI_Art_Assets/M_UI_AccoladeFlare
graph {
 properties {blendMode=BLEND_Additive, materialDomain=MD_UI, numCustomizedUVs=1, pixelDepthOffsetMode=PDOM_Legacy, usedShadingModels=Default Lit}
 nodes [
  n1 = MaterialFunctionCall(
   UVs=TextureCoordinate()
   Radius=Multiply(A=Constant() {r=0.5}, B=ScalarParameter() {defaultValue=1, parameterName=Radius})
   Density=ScalarParameter() {parameterName=RadialDensity}
  ) {materialFunction=&RadialGradientExponential}
  n2 = TextureCoordinate() {vTiling=32}
  n3 = Constant2Vector() {g=16, r=0.5}
  n4 = ScalarParameter() {parameterName=TimeControl}
  n5 = Sine(None=Multiply(A=Time()) {constB=0.25})
  n6 = Clamp(None=LinearInterpolate(A=Constant() {r=0.9}, B=Constant() {r=1}, Alpha=StaticSwitchParameter(True=@n4, False=@n5) {parameterName=ControlTime}))
  n7 = Multiply(A=Constant() {r=0.45}, B=@n6)
  n8 = Multiply(A=Constant() {r=1.25})
  n9 = Add(
   A=@n1.RadialGradientExponential
   B=DotProduct(
    A=OneMinus(None=@n1.RadialGradientExponential)
    B=MaterialFunctionCall(UVs=@n2, CenterPosition=@n3, Radius=@n7, Density=@n8) {materialFunction=&RadialGradientExponential}.RadialGradientExponential
   )
  )
  n10 = OneMinus(None=@n6)
  n11 = Constant() {r=1}
  n12 = Time()
  n13 = StaticSwitchParameter(True=ScalarParameter() {parameterName=TimeControl}, False=Sine(None=Multiply(A=@n12) {constB=0.25})) {parameterName=ControlTime}
  n14 = LinearInterpolate(A=Constant() {r=0.075}, B=Constant() {r=0.1}, Alpha=@n13)
  n15 = MaterialFunctionCall(UVs=TextureCoordinate(), Radius=@n14, Density=LinearInterpolate(A=Constant() {r=1.666}, B=Constant() {r=2.333}, Alpha=@n13)) {materialFunction=&RadialGradientExponential}
  LinearInterpolate(
   A=VectorParameter() {parameterName=Color2, defaultValue=<rgba 1 0.5237 0.0335 1>}.RGB
   B=VectorParameter() {parameterName=Color1, defaultValue=<rgba 1 0.9085 0.6903 1>}.RGB
   Alpha=Add(
    A=MaterialFunctionCall(Alpha=@n9, Min=@n10, Max=@n11) {materialFunction=&SmoothStep}.Result
    B=Multiply(A=@n15.RadialGradientExponential, B=@n14) {constB=0.5}
   )
  )
  MaterialFunctionCall(UVs=TextureCoordinate(), Texture Scale=VectorParameter() {parameterName=Scale, defaultValue=<rgba 1 1 0 1>}.RGB) {materialFunction=&ScaleUVsByCenter}
  ScalarParameter() {defaultValue=1, parameterName=Grow}
  ScalarParameter() {defaultValue=1, parameterName=Flare}
 ]
}
