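# Translucent UI material, named as a backing for the Creative scoreboard, that builds a two-color radial gradient.
# Widget pixel size scales the normalized UVs and gradient radius to account for aspect ratio, feeding an exponential radial gradient function.
# The gradient blends two color parameters and, separately, falloff and hotspot alpha parameters.
# Power nodes with color and alpha intensity parameters also read the gradient, but neither lerp uses them.
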Material /Game/Items/Traps/Blueprints/Toys/UI_Art_Assets/M_UI_CreativeScoreboard_Backing
graph {
 properties {blendMode=BLEND_Translucent, materialDomain=MD_UI, pixelDepthOffsetMode=PDOM_Legacy, usedShadingModels=Default Lit}
 nodes [
  Constant() {r=1.5}
  n1 = Constant() {r=1}
  n2 = MaterialFunctionCall() {materialFunction=&GetUserInterfaceUV}
  n3 = MaterialFunctionCall(Float2=@n2.Pixel Size) {materialFunction=&BreakOutFloat2Components}
  n4 = Max(A=@n3.R)
  n5 = Divide(A=@n3.G, B=@n4)
  n6 = Constant2Vector()
  n7 = MaterialFunctionCall(
   UVs=Multiply(A=AppendVector(A=@n1, B=@n5), B=@n2.Normalized UV)
   CenterPosition=Constant2Vector()
   Radius=Divide(A=Distance(A=@n6, B=@n2.Pixel Size), B=@n4)
   Density=Constant() {r=1.666}
  ) {materialFunction=&RadialGradientExponential}
  LinearInterpolate(
   A=VectorParameter() {parameterName=Color2, defaultValue=<rgba 0 0.0194 0.3185 1>}.RGB
   B=VectorParameter() {parameterName=Color1, defaultValue=<rgba 0 0.0578 0.6038 1>}.RGB
   Alpha=@n7.RadialGradientExponential
  )
  LinearInterpolate(
   A=ScalarParameter() {defaultValue=0.75, parameterName=FallOffAlpha}
   B=ScalarParameter() {defaultValue=1, parameterName=HotSpotAlpha}
   Alpha=@n7.RadialGradientExponential
  )
  Power(Base=@n7.RadialGradientExponential, Exp=ScalarParameter() {defaultValue=4, parameterName=GradientIntensity_Alpha}) {constExponent=4}
  Power(Base=@n7.RadialGradientExponential, Exp=ScalarParameter() {defaultValue=4, parameterName=GradientIntensity_Color}) {constExponent=4}
  TextureCoordinate()
 ]
}
