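# Translucent UI material for a scoreboard accolade, drawing an angled gold gradient panel with a border and animated shine.
# Rotated, offset UV ramps blend Highlight, Mid and Core colors into bands, while four more form the edge mask for the Border color.
# A time-scrolled sine stripe, gated by on/off cycle timing, color-dodges the interior when UseShineSwipe is set.
# UseOnlyBorder fades a separate solid mask toward just the border.
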Material /Game/Items/Traps/Blueprints/Toys/UI_Art_Assets/M_UI_Scoreboard_Accolade
graph {
 properties {blendMode=BLEND_Translucent, materialDomain=MD_UI, pixelDepthOffsetMode=PDOM_Legacy, usedShadingModels=Default Lit}
 nodes [
  n1 = TextureCoordinate()
  n2 = ScalarParameter() {defaultValue=-0.005, parameterName=BorderAngle_1}
  n3 = ComponentMask(None=Add(A=Rotator(Coordinate=@n1, Time=@n2) {speed=6.2832}, B=ScalarParameter() {defaultValue=-0.935, parameterName=BorderOffest_1})) {r=true}
  n4 = ScalarParameter() {defaultValue=128, parameterName=LR_Softness}
  n5 = Rotator(Coordinate=TextureCoordinate(), Time=ScalarParameter() {defaultValue=-0.005, parameterName=BorderAngle_2}) {speed=6.2832}
  n6 = ScalarParameter() {defaultValue=0.935, parameterName=BorderOffest_2}
  n7 = OneMinus(None=ComponentMask(None=Add(A=@n5, B=@n6)) {r=true})
  n8 = ScalarParameter() {defaultValue=128, parameterName=LR_Softness}
  n9 = Add(A=Clamp(None=Multiply(A=@n3, B=@n4)), B=Clamp(None=Multiply(A=@n7, B=@n8)))
  n10 = TextureCoordinate()
  n11 = ScalarParameter() {defaultValue=-0.245, parameterName=BorderAngle_3}
  n12 = ComponentMask(None=Add(A=Rotator(Coordinate=@n10, Time=@n11) {speed=6.2832}, B=ScalarParameter() {defaultValue=0.925, parameterName=BorderOffest_3})) {r=true}
  n13 = Clamp(None=Multiply(A=OneMinus(None=@n12), B=ScalarParameter() {defaultValue=48, parameterName=TB_Softness}))
  n14 = Rotator(Coordinate=TextureCoordinate(), Time=ScalarParameter() {defaultValue=-0.245, parameterName=BorderAngle_4}) {speed=6.2832}
  n15 = ScalarParameter() {defaultValue=-0.925, parameterName=BorderOffest_4}
  n16 = Multiply(A=ComponentMask(None=Add(A=@n14, B=@n15)) {r=true}, B=ScalarParameter() {defaultValue=48, parameterName=TB_Softness})
  n17 = Clamp(None=Add(A=Add(A=@n9, B=@n13), B=Clamp(None=@n16)))
  LinearInterpolate(A=Constant() {r=1}, B=@n17, Alpha=ScalarParameter() {parameterName=UseOnlyBorder})
  n18 = Rotator(Coordinate=TextureCoordinate(), Time=ScalarParameter() {defaultValue=-0.05, parameterName=GradAngle_1}) {speed=6.2832}
  n19 = ScalarParameter() {defaultValue=0.025, parameterName=Offest_1}
  n20 = ComponentMask(None=Multiply(A=Add(A=@n18, B=@n19), B=Constant() {r=8})) {r=true}
  n21 = Add(
   A=Rotator(Coordinate=TextureCoordinate(), Time=ScalarParameter() {defaultValue=-0.05, parameterName=GradAngle_1}) {speed=6.2832}
   B=ScalarParameter() {defaultValue=-0.075, parameterName=Offest_2}
  )
  n22 = Constant() {r=6}
  n23 = Clamp(None=ComponentMask(None=Multiply(A=@n21, B=@n22)) {r=true})
  n24 = LinearInterpolate(
   A=LinearInterpolate(
    A=VectorParameter() {parameterName=Mid, defaultValue=<rgba 1 0.6795 0.0369 1>}.RGB
    B=VectorParameter() {parameterName=Highlight, defaultValue=<rgba 1 1 1 1>}.RGB
    Alpha=Clamp(None=@n20)
   )
   B=LinearInterpolate(
    A=VectorParameter() {parameterName=Highlight, defaultValue=<rgba 1 1 1 1>}.RGB
    B=VectorParameter() {parameterName=Mid, defaultValue=<rgba 1 0.6795 0.0369 1>}.RGB
    Alpha=@n23
   )
   Alpha=@n23
  )
  n25 = TextureCoordinate()
  n26 = ScalarParameter() {defaultValue=-0.05, parameterName=GradAngle_1}
  n27 = Multiply(
   A=Add(A=Rotator(Coordinate=@n25, Time=@n26) {speed=6.2832}, B=ScalarParameter() {defaultValue=-0.4, parameterName=Offest_3})
   B=Constant() {r=4}
  )
  n28 = Clamp(None=Power(Base=ComponentMask(None=@n27) {r=true}) {constExponent=0.5})
  n29 = LinearInterpolate(
   A=VectorParameter() {parameterName=Mid, defaultValue=<rgba 1 0.6795 0.0369 1>}.RGB
   B=VectorParameter() {parameterName=Core, defaultValue=<rgba 0.3467 0.0953 0.0185 1>}.RGB
   Alpha=@n28
  )
  n30 = VectorParameter() {parameterName=Core, defaultValue=<rgba 0.3467 0.0953 0.0185 1>}
  n31 = VectorParameter() {parameterName=Mid, defaultValue=<rgba 1 0.6795 0.0369 1>}
  n32 = Add(
   A=Rotator(Coordinate=TextureCoordinate(), Time=ScalarParameter() {defaultValue=-0.05, parameterName=GradAngle_1}) {speed=6.2832}
   B=ScalarParameter() {defaultValue=-0.7, parameterName=Offest_4}
  )
  n33 = Constant() {r=3}
  n34 = Clamp(None=ComponentMask(None=Multiply(A=@n32, B=@n33)) {r=true})
  n35 = LinearInterpolate(
   A=LinearInterpolate(A=LinearInterpolate(A=@n24, B=@n29, Alpha=@n28), B=LinearInterpolate(A=@n30.RGB, B=@n31.RGB, Alpha=@n34), Alpha=@n34)
   B=VectorParameter() {parameterName=Border, defaultValue=<rgba 1 0.6795 0.0369 1>}.RGB
   Alpha=@n17
  )
  n36 = Constant()
  n37 = VectorParameter() {group=Shine, parameterName=ShineRGBA, defaultValue=<rgba 1 1 1 0.5>}
  n38 = ScalarParameter() {defaultValue=3, group=Shine, parameterName=Shine Speed}
  n39 = Multiply(A=Time(), B=@n38)
  n40 = Add(
   A=Rotator(
    Coordinate=MaterialFunctionCall() {materialFunction=&GetUserInterfaceUV}.9-Slice UV
    Time=ScalarParameter() {defaultValue=-0.05, group=Shine, parameterName=Shine Rotation}
   ) {centerX=0, centerY=0, speed=6.2832}
   B=AppendVector(A=OneMinus(None=@n39), B=Constant())
  )
  n41 = Clamp(None=Sine(None=ComponentMask(None=@n40) {r=true}) {period=2})
  n42 = Constant()
  n43 = Constant() {r=1}
  n44 = Constant()
  n45 = Time()
  n46 = ScalarParameter() {defaultValue=1, group=Shine, parameterName=Cycles On}
  n47 = Multiply(A=Divide(A=ScalarParameter() {defaultValue=6, group=Shine, parameterName=Cycles Off}, B=@n38), B=Constant() {r=2})
  n48 = Add(A=Multiply(A=Divide(A=@n46, B=@n38), B=Constant() {r=2}), B=@n47)
  n49 = Constant() {r=-1}
  n50 = Add(A=Fmod(A=@n45, B=Constant() {r=1}), B=If(A=Fmod(A=@n45, B=@n48), B=@n47, "A > B"=Constant(), "A == B"=@n49, "A < B"=@n49)) {constB=-1}
  n51 = LinearInterpolate(B=Multiply(A=LinearInterpolate(A=@n36, B=@n37.RGB, Alpha=@n41), B=If(A=@n50, B=@n42, "A > B"=@n43, "A < B"=@n44)), Alpha=OneMinus(None=@n17))
  LinearInterpolate(
   A=@n35
   B=Saturate(None=MaterialFunctionCall!(Blend=@n51, Base=@n35) {materialFunction=/Engine/Functions/Engine_MaterialFunctions03/Blends/Blend_ColorDodge}.Result)
   Alpha=ScalarParameter() {defaultValue=1, parameterName=UseShineSwipe, sliderMax=1}
  )
 ]
}
