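# Translucent UI material, named as the backing for the game winner on a scoreboard, that draws an angled panel shape with a gradient and sheen.
# Rotated, offset UV ramps combine into a resolution-corrected edge mask, with scalars to animate the center and right edges or disable angling.
# An aspect-corrected radial gradient blends two colors, warm gold and brown by default.
# A pannable, rotatable highlight band is color-dodged over that gradient.
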Material /Game/Items/Traps/Blueprints/Toys/UI_Art_Assets/M_UI_Scoreboard_GameWinnerBacking
graph {
 properties {blendMode=BLEND_Translucent, materialDomain=MD_UI, pixelDepthOffsetMode=PDOM_Legacy, usedShadingModels=Default Lit}
 nodes [
  Constant() {r=1.5}
  n1 = TextureCoordinate()
  n2 = ScalarParameter() {defaultValue=-0.245, parameterName=AngleLeft}
  n3 = ComponentMask(None=Add(A=Rotator(Coordinate=@n1, Time=@n2) {speed=6.2832}, B=ScalarParameter() {defaultValue=-0.025, parameterName=AngleLeftOffset})) {r=true}
  n4 = MaterialFunctionCall!(Native_U_PixelSize=ScalarParameter() {defaultValue=256, parameterName=NativeSize}) {materialFunction=/FortUILibrary/Foundation/Art_ToolKit/Tools/M_UI_Resolution_Sharpness_Correction}
  n5 = Multiply(A=@n4.Result, B=ScalarParameter() {defaultValue=80, parameterName=U_Softness})
  n6 = Rotator(Coordinate=TextureCoordinate(), Time=ScalarParameter() {defaultValue=-0.025, parameterName=AngleCenter}) {speed=6.2832}
  n7 = LinearInterpolate(
   A=Constant()
   B=ScalarParameter() {defaultValue=0.6666, parameterName=AngleCenterOffset}
   Alpha=ScalarParameter() {defaultValue=1, parameterName=AnimCenterAngle}
  )
  n8 = OneMinus(None=ComponentMask(None=Add(A=@n6, B=@n7)) {r=true})
  n9 = Multiply(A=@n4.Result, B=ScalarParameter() {defaultValue=128, parameterName=V_Softness})
  n10 = DotProduct(A=Clamp(None=Multiply(A=@n3, B=@n5)), B=Clamp(None=Multiply(A=@n8, B=@n9)))
  n11 = Rotator(Coordinate=TextureCoordinate(), Time=ScalarParameter() {defaultValue=-0.2375, parameterName=AngleRight}) {speed=6.2832}
  n12 = ScalarParameter() {defaultValue=0.85, parameterName=AngleRightOffset}
  n13 = OneMinus(None=ComponentMask(None=Add(A=@n11, B=@n12)) {r=true})
  n14 = Multiply(A=@n4.Result, B=ScalarParameter() {defaultValue=64, parameterName=U_Softness_R})
  n15 = OneMinus(None=Clamp(None=Multiply(A=@n13, B=@n14)))
  n16 = TextureCoordinate()
  n17 = ScalarParameter() {defaultValue=0.245, parameterName=AngleBottom}
  n18 = ComponentMask(None=Add(A=Rotator(Coordinate=@n16, Time=@n17) {speed=6.2832}, B=ScalarParameter() {defaultValue=-0.025, parameterName=AngleBottomOffset})) {r=true}
  n19 = Multiply(A=@n4.Result, B=ScalarParameter() {defaultValue=80, parameterName=U_Softness_Bottom})
  n20 = DotProduct(A=Clamp(None=Add(A=@n10, B=@n15)), B=Clamp(None=Multiply(A=@n18, B=@n19)))
  n21 = TextureCoordinate()
  n22 = ScalarParameter() {defaultValue=0.001, parameterName=LeftSideVAngle}
  n23 = ComponentMask(None=Add(A=Rotator(Coordinate=@n21, Time=@n22) {speed=6.2832}, B=ScalarParameter() {defaultValue=-0.025, parameterName=LeftSideOffset})) {r=true}
  n24 = Multiply(A=@n4.Result, B=ScalarParameter() {defaultValue=128, parameterName=L_Softness})
  n25 = Rotator(Coordinate=TextureCoordinate(), Time=ScalarParameter() {defaultValue=-0.0025, parameterName=RightSideVAngle}) {speed=6.2832}
  n26 = LinearInterpolate(
   A=Constant()
   B=ScalarParameter() {defaultValue=-0.965, parameterName=RightSideOffset}
   Alpha=ScalarParameter() {defaultValue=1, parameterName=AnimRightSide}
  )
  n27 = Multiply(
   A=ComponentMask(None=Add(A=@n25, B=@n26)) {r=true}
   B=Multiply(A=@n4.Result, B=ScalarParameter() {defaultValue=96, parameterName=R_Softness})
  )
  n28 = DotProduct(A=Clamp(None=Multiply(A=@n23, B=@n24)), B=OneMinus(None=Clamp(None=@n27)))
  LinearInterpolate(A=Constant() {r=1}, B=DotProduct(A=Clamp(None=@n20), B=Clamp(None=@n28)), Alpha=ScalarParameter() {defaultValue=1, parameterName=IsAngled})
  n29 = TextureCoordinate()
  n30 = Multiply(A=ScalarParameter() {defaultValue=-0.075, parameterName=Rotate}) {constB=0.125}
  n31 = ScalarParameter() {defaultValue=1.25, parameterName=Pan}
  n32 = ComponentMask(None=Add(A=Rotator(Coordinate=@n29, Time=@n30) {speed=6.2832}, B=@n31)) {r=true}
  n33 = Add(A=@n30, B=Constant() {r=0.5})
  n34 = ComponentMask(None=Add(A=Rotator(Coordinate=@n29, Time=@n33) {speed=6.2832}, B=OneMinus(None=@n31))) {r=true}
  n35 = Power(
   Base=LinearInterpolate(B=Power(Base=@n32) {constExponent=4}, Alpha=Power(Base=@n34) {constExponent=4})
   Exp=Multiply(A=ScalarParameter() {defaultValue=1, parameterName=Tightness}) {constB=16}
  ) {constExponent=16}
  n36 = MaterialFunctionCall() {materialFunction=&GetUserInterfaceUV}
  n37 = MaterialFunctionCall(Float2=@n36.Pixel Size) {materialFunction=&BreakOutFloat2Components}
  n38 = Max(A=@n37.R)
  n39 = Multiply(A=AppendVector(A=Constant() {r=1}, B=Divide(A=@n37.G, B=@n38)), B=@n36.Normalized UV)
  n40 = Constant2Vector()
  n41 = Constant() {r=1.666}
  n42 = Divide(A=Distance(A=Constant2Vector(), B=@n36.Pixel Size), B=@n38)
  MaterialFunctionCall!(
   Blend=Clamp(None=Clamp(None=@n35))
   Base=LinearInterpolate(
    A=VectorParameter() {parameterName=Color2, defaultValue=<rgba 0.3467 0.0953 0.0185 1>}.RGB
    B=VectorParameter() {parameterName=Color1, defaultValue=<rgba 1 0.5184 0.0369 1>}.RGB
    Alpha=MaterialFunctionCall(UVs=@n39, CenterPosition=@n40, Radius=@n42, Density=@n41) {materialFunction=&RadialGradientExponential}.RadialGradientExponential
   )
  ) {materialFunction=/Engine/Functions/Engine_MaterialFunctions03/Blends/Blend_ColorDodge}
  TextureCoordinate()
 ]
}
