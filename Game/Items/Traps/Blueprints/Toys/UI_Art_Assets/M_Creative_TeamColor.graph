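# M_Creative_TeamColor is a UI-domain material in the trap toys' UI art assets that supplies Creative team colors.
# It contains a single call to the shared MF_CreativeColors material function, which takes an Index input and has its Mask input fed a constant of 1.
# No connection from the function to a material output is shown.
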Material /Game/Items/Traps/Blueprints/Toys/UI_Art_Assets/M_Creative_TeamColor
graph {
 properties {materialDomain=MD_UI, pixelDepthOffsetMode=PDOM_Legacy, usedShadingModels=Default Lit}
 nodes [MaterialFunctionCall!(Mask=Constant() {r=1}, Index) {materialFunction=/Game/Creative/Materials/MF_CreativeColors}]
}
