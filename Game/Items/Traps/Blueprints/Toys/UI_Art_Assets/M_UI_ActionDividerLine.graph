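# Translucent UI material that procedurally draws a soft-edged, angled divider line shape, named for an action divider, without textures.
# Four UV rotators, driven by angle parameters in full turns, are offset, masked to red, scaled by edge softness and clamped to form half-plane edges.
# Scalar DotProducts multiply the left, right, top and bottom edges into one mask, alongside a Color parameter that defaults to white.
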Material /Game/Items/Traps/Blueprints/Toys/UI_Art_Assets/M_UI_ActionDividerLine
graph {
 properties {blendMode=BLEND_Translucent, materialDomain=MD_UI, pixelDepthOffsetMode=PDOM_Legacy, usedShadingModels=Default Lit}
 nodes [
  n1 = TextureCoordinate()
  n2 = Rotator(Coordinate=@n1, Time=ScalarParameter() {defaultValue=-0.095, parameterName=LeftAngle}) {speed=6.2832}
  n3 = ScalarParameter() {defaultValue=-0.335, parameterName=LeftOffset}
  n4 = Multiply(A=ComponentMask(None=Add(A=@n2, B=@n3)) {r=true}, B=ScalarParameter() {defaultValue=24, parameterName=EdgeSoftnessSides})
  n5 = Rotator(Coordinate=@n1, Time=ScalarParameter() {defaultValue=0.385, parameterName=RightAngle}) {speed=6.2832}
  n6 = ScalarParameter() {defaultValue=-0.4, parameterName=RightOffset}
  n7 = Multiply(A=ComponentMask(None=Add(A=@n5, B=@n6)) {r=true}, B=ScalarParameter() {defaultValue=24, parameterName=EdgeSoftnessSides})
  n8 = Rotator(Coordinate=@n1, Time=ScalarParameter() {defaultValue=-0.275, parameterName=TopAngle}) {speed=6.2832}
  n9 = ScalarParameter() {defaultValue=-0.025, parameterName=TopOffset}
  n10 = Multiply(A=ComponentMask(None=Add(A=@n8, B=@n9)) {r=true}, B=ScalarParameter() {defaultValue=64, parameterName=EdgeSoftnessTopsBottoms})
  n11 = Rotator(Coordinate=@n1, Time=ScalarParameter() {defaultValue=0.195, parameterName=BottomAngle}) {speed=6.2832}
  n12 = ScalarParameter() {parameterName=BottomOffset}
  n13 = Multiply(A=ComponentMask(None=Add(A=@n11, B=@n12)) {r=true}, B=ScalarParameter() {defaultValue=64, parameterName=EdgeSoftnessTopsBottoms})
  DotProduct(A=DotProduct(A=Clamp(None=@n4), B=Clamp(None=@n7)), B=DotProduct(A=Clamp(None=@n10), B=Clamp(None=@n13)))
  VectorParameter() {parameterName=Color, defaultValue=<rgba 1 1 1 1>}
 ]
}
