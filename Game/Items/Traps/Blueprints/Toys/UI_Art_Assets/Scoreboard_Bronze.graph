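# Scoreboard_Bronze is a purely visual UserWidget showing a bronze "3rd" place accolade badge, apparently for a toy scoreboard.
# A bronze backing material fills the root overlay, with a centered fixed-size SizeBox layering a slightly enlarged border and an accolade material image.
# Separate "3" and "rd" text blocks sit on top, offset and styled with bronze font materials and dark outlines.
# The event graph has only empty stub events, so the text is static.
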WidgetBlueprint /Game/Items/Traps/Blueprints/Toys/UI_Art_Assets/Scoreboard_Bronze
graphs [EventGraph {nodes [
   Event PreConstruct()
   Event Construct()
   Event Tick()
  ]}]
variables []
widgets {
 info {inheritedWidgetCount=0, parentClass=&UserWidget, rootWidgetClass=&Overlay, widgetCount=8}
 widgets [
  Overlay_8 {widgetClassPath=&Overlay, widget=Overlay_8}
  Image_47 {
   widgetClassPath=&Image
   bIsVariable=true
   parent=Overlay_8
   widget=Image_47
   properties {brush {
     bIsDynamicallyLoaded=false
     drawAs=Image
     imageType=NoImage
     mirroring=NoMirror
     resourceName=None
     resourceObject=&/M_UI_Accolade_Backing_Bronze
     tiling=NoTile
     imageSize=<32 32>
     margin=<ltrb 0 0 0 0>
     outlineSettings {
      bUseBrushTransparency=false
      roundingType=HalfHeightRadius
      width=0
      color {colorUseRule=UseColor_Specified, specifiedColor=<rgba 0 0 0 0>}
      cornerRadii=<0 0 0 0>
     }
     tintColor {colorUseRule=UseColor_Specified, specifiedColor=<rgba 1 1 1 1>}
     uVRegion {bIsValid=false, max=<0 0>, min=<0 0>}
    }}
   slot {class=OverlaySlot, properties {horizontalAlignment=HAlign_Fill, verticalAlignment=VAlign_Fill}}
  }
  SizeBox_0 {
   widgetClassPath=&SizeBox
   parent=Overlay_8
   widget=SizeBox_0
   properties {bOverride_HeightOverride=true, bOverride_WidthOverride=true, heightOverride=32, widthOverride=64}
   slot {class=OverlaySlot, properties {horizontalAlignment=HAlign_Center, verticalAlignment=VAlign_Center}}
  }
  Overlay_57 {widgetClassPath=&Overlay, parent=SizeBox_0, widget=Overlay_57, slot {class=SizeBoxSlot}}
  Image_167 {
   widgetClassPath=&Image
   bIsVariable=true
   parent=Overlay_57
   widget=Image_167
   properties {
    brush {
     bIsDynamicallyLoaded=false
     drawAs=Image
     imageType=NoImage
     mirroring=NoMirror
     resourceName=None
     resourceObject=&/M_UI_AccoladeBorder_Bronze
     tiling=NoTile
     imageSize=<32 32>
     margin=<ltrb 0.75 0.75 0.75 0.75>
     outlineSettings {
      bUseBrushTransparency=false
      roundingType=HalfHeightRadius
      width=0
      color {colorUseRule=UseColor_Specified, specifiedColor=<rgba 0 0 0 0>}
      cornerRadii=<0 0 0 0>
     }
     tintColor {colorUseRule=UseColor_Specified, specifiedColor=<rgba 1 1 1 1>}
     uVRegion {bIsValid=false, max=<0 0>, min=<0 0>}
    }
    renderTransform {angle=0, scale=<1.1 1.2>, shear=<0 0>, translation=<0 0>}
   }
   slot {class=OverlaySlot, properties {horizontalAlignment=HAlign_Fill, verticalAlignment=VAlign_Fill}}
  }
  Image_96 {
   widgetClassPath=&Image
   bIsVariable=true
   parent=Overlay_57
   widget=Image_96
   properties {brush {
     bIsDynamicallyLoaded=false
     drawAs=Image
     imageType=NoImage
     mirroring=NoMirror
     resourceName=None
     resourceObject=&/M_UI_Scoreboard_Accolade_Bronze
     tiling=NoTile
     imageSize=<32 32>
     margin=<ltrb 0 0 0 0>
     outlineSettings {
      bUseBrushTransparency=false
      roundingType=HalfHeightRadius
      width=0
      color {colorUseRule=UseColor_Specified, specifiedColor=<rgba 0 0 0 0>}
      cornerRadii=<0 0 0 0>
     }
     tintColor {colorUseRule=UseColor_Specified, specifiedColor=<rgba 1 1 1 1>}
     uVRegion {bIsValid=false, max=<0 0>, min=<0 0>}
    }}
   slot {class=OverlaySlot, properties {horizontalAlignment=HAlign_Fill, verticalAlignment=VAlign_Fill}}
  }
  CommonTextBlock_85 {
   widgetClassPath=&CommonTextBlock
   parent=Overlay_57
   widget=CommonTextBlock_85
   properties {
    text=3
    font {
     bForceMonospaced=false
     bMaterialIsStencil=false
     fontMaterial=&/M_UI_AccoladeText_Bronze
     fontObject=&Burbank
     letterSpacing=0
     monospacedWidth=1
     size=22
     skewAmount=0
     typefaceFontName=Burbank-Black
     outlineSettings {
      bApplyOutlineToDropShadows=false
      bMiteredCorners=false
      bSeparateFillAlpha=false
      outlineMaterial=null
      outlineSize=2
      outlineColor=<rgba 0.1119 0.0137 0.0194 1>
     }
    }
    renderTransform {angle=0, scale=<1 1>, shear=<-16 0>, translation=<-2 -4>}
    shadowOffset=<0 0>
   }
   slot {class=OverlaySlot, properties {horizontalAlignment=HAlign_Center, verticalAlignment=VAlign_Center}}
  }
  CommonTextBlock_87 {
   widgetClassPath=&CommonTextBlock
   parent=Overlay_57
   widget=CommonTextBlock_87
   properties {
    text=rd
    font {
     bForceMonospaced=false
     bMaterialIsStencil=false
     fontMaterial=&M_UI_AccoladeText_Bronze
     fontObject=&BurbankSmall
     letterSpacing=0
     monospacedWidth=1
     size=9
     skewAmount=0
     typefaceFontName=Black
     outlineSettings {
      bApplyOutlineToDropShadows=false
      bMiteredCorners=false
      bSeparateFillAlpha=false
      outlineMaterial=null
      outlineSize=2
      outlineColor=<rgba 0.1119 0.0137 0.0194 1>
     }
    }
    renderTransform {angle=0, scale=<1 1>, shear=<0 0>, translation=<8 2>}
    shadowOffset=<0 0>
   }
   slot {class=OverlaySlot, properties {horizontalAlignment=HAlign_Center, verticalAlignment=VAlign_Center}}
  }
 ]
}
